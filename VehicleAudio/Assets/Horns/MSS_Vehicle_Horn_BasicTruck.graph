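# MetaSound Source preset of MSS_Vehicle_Horn that supplies the basic truck horn sound, using the Valet V8 truck horn wave.
# All graph logic lives in the parent, which exposes play and stop triggers, a looping flag, output volume and decay time and curve, here set to a short decay.
# The mono looping source uses vehicle horn attenuation, a cosmetic horn property sheet and the exterior vehicles sound class.
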MetaSoundSource /VehicleAudio/Assets/Horns/MSS_Vehicle_Horn_BasicTruck
preset_parent=/VehicleAudio/Patches/MSS_Vehicle_Horn
inputs [
 DecayCurve {type=Float}
 DecayTime {type=Time}
 Looping {type=Bool}
 OnStop {type=Trigger}
 OutputVolume {type=Float, access=Constructor}
 UE.Source.OnPlay {type=Trigger}
 Wave Asset {type=WaveAsset}
]
interfaces [UE.OutputFormat.Mono:1.0, UE.Source:1.1]
outputs ["UE.OutputFormat.Mono.Audio:0" {type=Audio}]
pages [Default {input_defaults {
   DecayCurve=1.5
   DecayTime=0.1
   Looping=null
   OnStop=null
   OutputVolume=0
   UE.Source.OnPlay=null
   Wave Asset=&/Veh_Valet_Truck_V8_01_Horn_01
  }}]
properties {
 assetImportData=&self:AssetImportData
 attenuationSettings=/VehicleAudio/Settings/Vehicle_Horn_Att
 audioPropertiesSheet=/VehicleAudio/PropertySheets/APS_Cosmetic_Vehicle_Horn
 bLooping=true
 duration=10000
 maxDistance=16030.001
 sampleRate=48000
 soundClassObject=&Fort_Vehicles_Exterior
 volume=1.585
}
